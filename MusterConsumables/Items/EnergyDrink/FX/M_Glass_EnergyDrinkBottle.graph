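# Translucent, surface-lit glass material for the energy drink bottle consumable's FX, usable on Niagara mesh and sprite particles.
# An inverted dot product of the normal map and a constant vector forms a fake directional mask, blending particle-color brightness and an opacity-like term.
# That term is scaled by diffuse red and particle alpha.
# Diffuse times particle color feeds MF_FX_EmissiveAdjustments; refraction uses index-of-refraction mode, with screen-space reflections enabled.
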Material /MusterConsumables/Items/EnergyDrink/FX/M_Glass_EnergyDrinkBottle
graph {
 properties {
  bScreenSpaceReflections=true
  bUsedWithNiagaraMeshParticles=true
  bUsedWithNiagaraSprites=true
  blendMode=BLEND_Translucent
  pixelDepthOffsetMode=PDOM_Legacy
  refractionMethod=RM_IndexOfRefraction
  translucencyLightingMode=TLM_Surface
  translucencyPass=MTP_BeforeDOF
  usedShadingModels=Default Lit
 }
 nodes [
  Constant() {r=1}
  Constant() {r=5}
  n1 = ParticleColor()
  n2 = OneMinus(None=DotProduct(
    A=Constant3Vector() {constant=<rgba 2 0 1 0>}
    B=TextureSample() {samplerType=SAMPLERTYPE_Normal, texture=/MusterConsumables/Items/EnergyDrink/Items/Textures/T_Gadget_Energy_Drink_N}.RGB
   ))
  LinearInterpolate(A=Multiply(A=@n1.RGB) {constB=1.2}, B=Multiply(A=@n1.RGB) {constB=100}, Alpha=@n2) {constA=1.2, constB=100}
  n3 = TextureSample() {texture=/MusterConsumables/Items/EnergyDrink/Items/Textures/T_Gadget_Energy_Drink_D}
  n4 = ParticleColor()
  MaterialFunctionCall(Emissive=Multiply(A=@n3.RGB, B=@n4.RGB)) {materialFunction=&MF_FX_EmissiveAdjustments}
  Multiply(A=Multiply(A=LinearInterpolate(Alpha=@n2) {constA=0.15}, B=@n3.R), B=@n4.A)
 ]
}
