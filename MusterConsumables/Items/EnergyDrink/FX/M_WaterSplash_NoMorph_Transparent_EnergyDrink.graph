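# Translucent unlit water splash material for the Energy Drink consumable's FX, usable on Niagara mesh particles.
# Emissive adds a contrast-adjusted reflection texture, sampled by reflection vector, to a base color, then runs through an FX adjustment function.
# Opacity combines that reflection with a height mask above the object's origin, scaled by particle alpha.
# A noise texture offset by the dynamic Erode parameter dissolves it, and depth fade softens intersections.
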Material /MusterConsumables/Items/EnergyDrink/FX/M_WaterSplash_NoMorph_Transparent_EnergyDrink
graph {
 properties {
  bTangentSpaceNormal=false
  bUsedWithNanite=true
  bUsedWithNiagaraMeshParticles=true
  blendMode=BLEND_Translucent
  pixelDepthOffsetMode=PDOM_Legacy
  shadingModel=MSM_Unlit
  usedShadingModels=Unlit
 }
 nodes [
  ActorPositionWS()
  Constant() {r=0.5}
  Constant() {r=1}
  Constant()
  n1 = ParticleColor()
  n2 = TextureSampleParameter2D(UVs=ComponentMask(None=ReflectionVectorWS()) {g=true, r=true}) {parameterName=Reflect Texture, texture=/Game/Items/Art_noLOD/Textures/T_Items_MineralPowder_Glass_Reflection}
  n3 = Add(A=@n1.A) {constB=0.15}
  n4 = WorldPosition()
  n5 = TransformPosition(None=Constant3Vector()) {transformType=TRANSFORMPOSSOURCE_World}
  n6 = Divide(A=Subtract(
    A=MaterialFunctionCall(Float3=@n4.XYZ) {materialFunction=&BreakOutFloat3Components}.B
    B=MaterialFunctionCall(Float3=@n5) {materialFunction=&BreakOutFloat3Components}.B
   )) {constB=25}
  n7 = Add(
   A=Multiply(A=Multiply(A=@n2.G, B=@n3), B=ScalarParameter() {defaultValue=1, parameterName=Use Reflect Texture in opac})
   B=LinearInterpolate(B=Clamp(None=@n6), Alpha=ScalarParameter() {defaultValue=1, parameterName=Use World Mask}) {constA=1}
  )
  n8 = Add(
   A=TextureSample() {samplerType=SAMPLERTYPE_LinearColor, texture=/Game/Global/Textures/Default/Noise/T_Noise01}.G
   B=DynamicParameter() {paramNames [Erode,Param2,Param3,Param4]}.Erode
  )
  DepthFade(Opacity=Multiply(A=Multiply(A=@n1.A, B=@n7), B=Clamp(None=@n8)), FadeDistance=ScalarParameter() {defaultValue=1, parameterName=DepthFade})
  LinearInterpolate(B=Constant() {r=1.4}, Alpha=@n1.R) {constA=1}
  n9 = ScalarParameter() {defaultValue=1, parameterName=Add reflect texture to base color}
  n10 = MaterialFunctionCall(In=@n2.G, Contrast=ScalarParameter() {defaultValue=1, parameterName=REflect tex contrast}) {materialFunction=&CheapContrast}
  MaterialFunctionCall(Emissive=Add(A=Multiply(A=@n9, B=@n10.Result), B=VectorParameter() {parameterName=BaseColor, defaultValue=<rgba 0.2026 0.3282 0.401 0>}.RGB)) {materialFunction=&MF_FX_EmissiveAdjustments}
  Reroute(None=Constant() {r=0.35})
  ScalarParameter() {parameterName=Z SpawnHeight}
  ScalarParameter() {parameterName=alphatest}
 ]
}
