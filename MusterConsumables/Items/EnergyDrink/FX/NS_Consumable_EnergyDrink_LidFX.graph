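# Niagara effect for the energy drink consumable that pops a single short-lived mesh cap upward, likely the can's lid.
# One Cap emitter bursts one mesh particle with upward velocity, then applies gravity, fading rotational force and bouncy CPU collision on the Pawn channel.
# The cap fades alpha over its life and shrinks near the end, with distance culling on the emitter.
# User color parameters are declared, but no shown module reads them.
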NiagaraSystem /MusterConsumables/Items/EnergyDrink/FX/NS_Consumable_EnergyDrink_LidFX
dependencies {
 usedDataInterfaces [&NiagaraDataInterfaceCurve, &NiagaraDataInterfaceVectorCurve]
 usedDynamicInputs []
 usedModules [&SystemState]
}
emitters {Cap {
  emitterUpdateScript [
   &EmitterState(Enable Distance Culling=true, MaxDistance=4000, Module.bUseMaxDistance=true, Module.bUseMinDistance=false, Scalability Mode=::NewEnumerator1)
   &SpawnBurst_Instantaneous(Spawn Count=1, Use Spawn Probability=false)
  ]
  particleSpawnScript [
   &InitializeParticle(Lifetime=0.35, Mesh Scale Mode=::NewEnumerator3, UsePositionOffset=false, Write Material Random=true) {module=InitializeParticle001}
   &InitialMeshOrientation(Mesh Orientation Mode=::NewEnumerator1, Use Rotation Vector=false)
   &AddVelocity(Velocity=<0 0 500>) {module=AddVelocity001}
   /Niagara/Modules/Update/Mass/CalculateSizeAndRotationalInertiaByMass(Depth=5, Height=2, Module.Manually Enter Density=false, Width=5)
   &ApplyInitialForces()
  ]
  particleUpdateScript [
   &ParticleState()
   /Niagara/Modules/Update/Forces/FindKineticAndPotentialEnergy()
   &ScaleMeshSize(Scale Factor=&VectorFromFloat(Value=&FloatFromCurve(FloatCurve=DataInterface() {propertyValues {Curve {keys [<at 0.8 1>,<at 1 0>]}}})))
   &UpdateMeshOrientation() {enabled=false}
   &GravityForce()
   &Drag(Drag=&UniformRangedFloat(Minimum=2, Maximum=3), Use Linear Drag=true, Use Rotational Drag=false) {enabled=false}
   &MeshRotationForce(Rotation=&VectorFromCurve(
     VectorCurve=DataInterface() {propertyValues {XCurve {keys [<at 0 1>,<at 1 0>]}, YCurve {keys [<at 0 1>,<at 1 0>]}, ZCurve {keys [<at 0 1>,<at 1 0>]}}}
     Scale Curve=<50 50 50>
    ))
   &ScaleColor(Scale Alpha=&Lerp_Float(A=1, B=0, Alpha=$Particles.NormalizedAge), ScaleA=true, ScaleRGB=true)
   &Collision(
    Age Colliding Particles=false
    CPU Collision Trace Channel=ECC_Pawn
    Control Roll On Collision=true
    Mesh Dimensions:Vector3f=<5 5 2>
    Method for Calculating Particle Radius:ENiagaraCalculateRadiusOptions=::NewEnumerator1
    Module.Control Rest State Via Surface Normal=false
    Module.EnableMaxCollisionCount=false
    Module.Randomize Collision Normal Vector=false
    Radius Calculation Type=::NewEnumerator1
    Restitution=1
    ScaleRateOfRotation=false
   )
   &SolveForcesAndVelocity(Clamp Velocity=false, Limit Acceleration=false)
   &ScaleSpriteSizeBySpeed(Max Scale Factor=<0.25 2>, Min Scale Factor=<0.15 0.15>)
  ]
  rendererClasses [&NiagaraMeshRendererProperties]
  renderers [{rendererClass=&NiagaraMeshRendererProperties, rendererIndex=0}]
 }}
summary {
 systemName=NS_Consumable_EnergyDrink_LidFX
 emitters [{emitterName=Cap, rendererClasses [&NiagaraMeshRendererProperties]}]
 userVariables [
  User.ColorA {type=LinearColor, description="", defaultValue {struct=LinearColor, value=<rgba 1 0.9123 0 1>}}
  User.ColorB {type=LinearColor, description="", defaultValue {struct=LinearColor, value=<rgba 1 0.6006 0 1>}}
  User.ColorC {type=LinearColor, description="", defaultValue {struct=LinearColor, value=<rgba 5 2.4353 0 1>}}
  User.ColorGradient {
   type=NiagaraDataInterfaceColorCurve
   description=""
   defaultValue {
    struct=NiagaraExt_VariableValue_DataInterface
    value {
     dataInterface=&self:NiagaraDataInterfaceColorCurve_0
     dataInterfaceClass=&NiagaraDataInterfaceColorCurve
    }
   }
  }
 ]
}
