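# Niagara system that renders the liquid in the Energy Drink consumable, apparently while drinking, through one mesh emitter named Shield_Potion_Liquid.
# A spring-driven pendulum, fed by owner movement and rotation velocity, fakes slosh and orients the mesh toward its pivot.
# Tilt and travel speed become smoothed liquid speed and turbulence.
# Dynamic material parameters carry these, mesh axes, a time counter and a timeline-blended curve value, likely fill level.
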NiagaraSystem /MusterConsumables/Items/EnergyDrink/FX/NS_EnergyDrink_LiquidDrinking
dependencies {
 usedDataInterfaces [
  &NiagaraDataInterfaceCurve
  &NiagaraDataInterfaceMeshRendererInfo
  &NiagaraDataInterfaceVectorCurve
 ]
 usedDynamicInputs []
 usedModules [&SystemState]
}
emitters {Shield_Potion_Liquid {
  emitterUpdateScript [
   &EmitterState(Inactive Response=::NewEnumerator1, Life Cycle Mode=::NewEnumerator1, Loop Behavior=::NewEnumerator1, UseLoopDelay=false)
   &SpawnBurst_Instantaneous(Use Spawn Probability=false, UseLoopCountLimit=false)
   /SetVariables(Emitter.Orientation:Quat4f=$Engine.Owner.Rotation) {isSetParameters=true}
   /SetVariables(Emitter.Position:Vector3f=/Niagara/DynamicInputs/Helpers/SimulationPosition()) {isSetParameters=true}
   /SetVariables(
    Emitter.RotationVelocity:Vector3f=&Divide_Vector(
     A:Vector3f=&Subtract_Vector(
      A=&RotateVectorByQuaternion(Vector To Rotate=<0 0 1>, Quaternion=$Emitter.Orientation)
      B=&RotateVectorByQuaternion(Vector To Rotate=<0 0 1>, Quaternion=$Emitter.Previous.Orientation)
     )
     B:Vector3f=&VectorFromFloat(Value=$Engine.DeltaTime)
    )
    Emitter.Velocity:Vector3f=&Divide_Vector(
     A:Vector3f=&Subtract_Vector(A=$Emitter.Position, B=$Emitter.Previous.Position)
     B:Vector3f=&VectorFromFloat(Value=$Engine.DeltaTime)
    )
    Emitter.Viscosity=1
   ) {isSetParameters=true}
   /SetVariables(
    Emitter.ColorA:LinearColor=<rgba 0.0936 0.3021 0.0302 1>
    Emitter.ColorB:LinearColor=<rgba 0.0936 0.3021 0.0302 1>
    Emitter.Debug Draw:Vector3f=<10 10 10>
   ) {isSetParameters=true}
  ]
  particleSpawnScript [
   &InitializeParticle(Color Mode=::NewEnumerator0, Mesh Scale Mode=::NewEnumerator4, UsePositionOffset=false, Write Material Random=true)
   &InitialMeshOrientation(Mesh Orientation Mode=::NewEnumerator1, Use Rotation Vector=false) {enabled=false}
  ]
  particleUpdateScript [
   &ParticleState(Kill Particles When Lifetime Has Elapsed=false, Let Infinitely Lived Particles Die When Emitter Deactivates=true)
   &ScaleMeshSize(Scale Factor=&VectorFromCurve(
     VectorCurve=DataInterface() {propertyValues {
       XCurve {keys [<at 0.8 1>,<at 0.9 0>,<at 1 1 RCIM_Cubic>]}
       YCurve {keys [<at 0.8 1>,<at 0.9 0>,<at 1 1 RCIM_Cubic>]}
       ZCurve {keys [<at 0.8 1>,<at 0.9 0 RCIM_Cubic>,<at 1 1>]}
      }}
     CurveIndex=$User.BottleSqueeze
    )) {enabled=false}
   &ShapeLocation(Apply To Particle Position=::NewEnumerator1, MaskSpawnGroup=false, Override Seed=false, UseOwnerScale=true) {enabled=false}
   /SetVariables(Particles.Position_Origin:NiagaraPosition=$Engine.Emitter.SimulationPosition) {isSetParameters=true}
   &Drag(Drag=0, Use Linear Drag=true, Use Rotational Drag=false)
   &PendulumSetup(
    Module.Enable Spring Driver=true
    Pendulum Length=5
    Pendulum Pivot:NiagaraPosition=$Particles.Position_Origin
    Rigid Pendulum=false
    "Spring Driven Constraint (Tightness)"=1
   )
   &InheritSourceMovement(
    Applied Position Scale=&Multiply_VectorByFloat(Vector=<1 1 1>, Float=1)
    ApplyForce=false
    ApplyPosition=true
    ApplyPositionFalloff=false
    Source Velocity:Vector3f=$Emitter.RotationVelocity
   )
   &InheritVelocity(
    Inherited Velocity Amount Scale=&Multiply_VectorByFloat(Vector=<1 1 1>, Float=-0.06)
    Speed Limit=false
    UseSpeedThreshold=false
   )
   &SolveForcesAndVelocity(Clamp Velocity=false, Limit Acceleration=false, Speed Limit=200)
   &PendulumConstraint()
   &CalculateAccurateVelocity()
   /Niagara/Modules/Update/Orientation/V2/OrientMeshToVector(
    Look At Direction:Vector3f=&Subtract_Vector(
     A=Linked() {linkedVariable {name=Particles.Position_Origin, type=NiagaraPosition}}
     B=Linked() {linkedVariable {name=Particles.Position, type=NiagaraPosition}}
    )
    Override Quat=false
    Reference Axis Vector:Vector3f=<0 0 1>
    UseOptionalSideAxis=false
    UseOptionalUpAxis=false
   )
   &InterpolateOverTime(
    Initial Float Value:NiagaraFloat=&MakeCustomFloatFromBool(Bool=$User.IsPickup, True Float=1, False Float=0)
    Rate Of Change=3
    Target Float Value:NiagaraFloat=&Max_Float(
     A:NiagaraFloat=&RemapRange(
      Input Value To Remap=&DotProduct(
       Vector A=<0 0 -1>
       Vector B=&ReturnMeshOrientationAxis(Mesh Orientation Axis=::ZAxis, Mesh Orientation Quat=$Particles.MeshOrientation)
      )
      Input Min=-1
      Input Max=0
      Output Min=0
      Output Max=1
     )
     B:NiagaraFloat=&RemapRange(
      Input Value To Remap=&Divide_Float(A=&Subtract_Float(A=$Particles.DistanceTraveled, B=$Particles.Previous.DistanceTraveled), B=$User.LiquidScale)
      Input Min=2
      Input Max=4
      Output Min=0
      Output Max=1
     )
    )
   )
   /SetVariables(Particles.LiquidSpeed:NiagaraFloat=&ClampFloat(Float=$Particles.InterpolateOverTime.InterpolateOverTime_Float.Moving Average)) {isSetParameters=true}
   &Timeline(Play=&SetBoolByFloatComparison(A=$Particles.LiquidSpeed, B=0.333), Play Rate=3, Timeline End=10)
   /SetVariables(Particles.Turbulence:NiagaraFloat=$Particles.Timeline.NormalizedTime) {isSetParameters=true}
   /Niagara/Modules/Update/Utility/IncrementOverTime(Rate of Change:NiagaraFloat=&Lerp_Float(A=1, B=1.25, Alpha=$Particles.LiquidSpeed))
   /SetVariables(
    Particles.DynamicMaterialParameter:Vector4f=&MakeFloat4FromVectorAndFloat(
     "Vector (RGB)"=&ReturnMeshOrientationAxis(Mesh Orientation Axis=::ZAxis, Mesh Orientation Quat=$Particles.MeshOrientation)
     "Float (Alpha)"=&Lerp_Float(
      A=0.75
      B=&FloatFromCurve(FloatCurve=DataInterface() {propertyValues {Curve {keys [<at 0.5 0.5>,<at 1.5032 0.5>,<at 2 -2>]}}}, CurveIndex=$Emitter.Age)
      Alpha=$System.Timeline.NormalizedTime
     )
    )
    Particles.DynamicMaterialParameter1:Vector4f=&MakeFloat4FromVectorAndFloat(
     "Vector (RGB)"=&ReturnMeshOrientationAxis(Mesh Orientation Axis=::XAxis, Mesh Orientation Quat=$Particles.MeshOrientation)
     "Float (Alpha)"=$Particles.IncrementOverTime.Counter
    )
    Particles.DynamicMaterialParameter2:Vector4f=&MakeVector4(X=$Particles.LiquidSpeed, Y=$Particles.Turbulence, Z=0, W=0)
   ) {isSetParameters=true}
  ]
  rendererClasses [&NiagaraMeshRendererProperties]
  renderers [
   {rendererClass=&NiagaraMeshRendererProperties, rendererIndex=0}
   {rendererClass=&NiagaraMeshRendererProperties, rendererIndex=1}
  ]
 }}
summary {
 systemName=NS_EnergyDrink_LiquidDrinking
 emitters [{emitterName=Shield_Potion_Liquid, rendererClasses [&NiagaraMeshRendererProperties]}]
 userVariables [
  User.BottleSqueeze {type=NiagaraFloat, defaultValue=1, description=""}
  User.BowlRadius {type=NiagaraFloat, defaultValue=0, description=""}
  User.Bubbles Min Speed {type=NiagaraFloat, defaultValue=0, description=""}
  User.DrinkingScale {type=NiagaraFloat, defaultValue=0, description=""}
  User.Duration {type=NiagaraFloat, defaultValue=1, description=""}
  User.IsDrinking {type=NiagaraBool, defaultValue=true}
  User.IsFinished {type=NiagaraBool}
  User.IsPickup {type=NiagaraBool, defaultValue=true}
  User.LiquidScale {type=NiagaraFloat, defaultValue=1, description=""}
  User.PlayAudio {type=NiagaraBool}
 ]
}
