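# Sound cue for the cook sound of the Energy Drink consumable, built from bottle reach-back recordings.
# A Random node picks one of two equally weighted waves, and a Modulator applies a slight random upward pitch shift for variation.
# The cue plays at reduced volume in the gadgets sound class, using weapon reload attenuation for distance falloff.
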SoundCue /MusterConsumables/Items/EnergyDrink/Sounds/EnergyDrink_Cook_Cue
root=Modulator_0
nodes [
 {id=Modulator_0, children [Random_3], properties {pitchMax=1.1, pitchMin=1}}
 {id=Random_3, children [WavePlayer_7,WavePlayer_8], properties {weights [1,1]}}
 {id=WavePlayer_7, properties {soundWaveAssetPtr=&/Energy_Bottle_ReachBack_02}}
 {id=WavePlayer_8, properties {soundWaveAssetPtr=&/Energy_Bottle_ReachBack_03}}
]
properties {
 attenuationSettings=&Fort_Weapon_Reload_Att
 maxDistance=2250
 soundClassObject=&Fort_Gadgets
 volumeMultiplier=0.4
}
