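# Sound cue for the first-person discard of the Energy Drink consumable, playing a bottle impact sound.
# A random node picks one of three equally weighted impact waves, which plays after a fixed 0.6-second delay and feeds a reduced-volume mixer whose other input is empty.
# The cue uses the gadgets sound class and borrows weapon-reload attenuation.
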SoundCue /MusterConsumables/Items/EnergyDrink/Sounds/EnergyDrink_Discard_1P_Cue
root=Mixer_0
nodes [
 {id=Delay_0, children [Random_3], properties {delayMax=0.6, delayMin=0.6}}
 {id=Mixer_0, children [null,Delay_0], properties {inputVolume [0.7,0.7]}}
 {id=Random_3, children [WavePlayer_4, WavePlayer_5, WavePlayer_10], properties {weights [1,1,1]}}
 {id=WavePlayer_10, properties {soundWaveAssetPtr=&/Energy_Bottle_DiscardImpact_03}}
 {id=WavePlayer_4, properties {soundWaveAssetPtr=&/Energy_Bottle_DiscardImpact_01}}
 {id=WavePlayer_5, properties {soundWaveAssetPtr=&/Energy_Bottle_DiscardImpact_02}}
]
properties {
 attenuationSettings=&Fort_Weapon_Reload_Att
 maxDistance=2250
 soundClassObject=&Fort_Gadgets
 volumeMultiplier=0.85
}
