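# Sound cue for discarding the energy drink consumable's bottle.
# A Random node picks one of three equally weighted discard waves and feeds a Modulator at default settings.
# That feeds a Mixer with a reduced input volume and an empty second input, and the cue's volume multiplier lowers the level further.
# It borrows the weapon reload attenuation and plays under the gadgets sound class.
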SoundCue /MusterConsumables/Items/EnergyDrink/Sounds/EnergyDrink_Discard_Cue
root=Mixer_0
nodes [
 {id=Mixer_0, children [Modulator_0,null], properties {inputVolume [0.7,0.7]}}
 {id=Modulator_0, children [Random_0]}
 {id=Random_0, children [WavePlayer_0,WavePlayer_1,WavePlayer_3], properties {weights [1,1,1]}}
 {id=WavePlayer_0, properties {soundWaveAssetPtr=&/Energy_Bottle_Discard_01}}
 {id=WavePlayer_1, properties {soundWaveAssetPtr=&/Energy_Bottle_Discard_02}}
 {id=WavePlayer_3, properties {soundWaveAssetPtr=&/Energy_Bottle_Discard_03}}
]
properties {
 attenuationSettings=&Fort_Weapon_Reload_Att
 maxDistance=2250
 soundClassObject=&Fort_Gadgets
 volumeMultiplier=0.6
}
