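# Sound cue for the impact sound when an Energy Drink consumable is dropped, using bottle drop impact recordings.
# A Random node picks one of three equally weighted waves and feeds a Modulator that slightly varies pitch.
# It uses the weapon-drop attenuation and the gadgets/items sound class at reduced volume.
# A concurrency override prevents new instances from starting once the limit is reached.
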SoundCue /MusterConsumables/Items/EnergyDrink/Sounds/EnergyDrink_DropImpact_Cue
root=Modulator_1
nodes [
 {id=Modulator_1, children [Random_1], properties {pitchMax=1.1, pitchMin=0.9}}
 {id=Random_1, children [WavePlayer_0,WavePlayer_3,WavePlayer_4], properties {weights [1,1,1]}}
 {id=WavePlayer_0, properties {soundWaveAssetPtr=&/Energy_Bottle_DropImpact_01}}
 {id=WavePlayer_3, properties {soundWaveAssetPtr=&/Energy_Bottle_DropImpact_02}}
 {id=WavePlayer_4, properties {soundWaveAssetPtr=&/Energy_Bottle_DropImpact_03}}
]
properties {
 attenuationSettings=&Fort_Weapon_Drop_Att
 bOverrideConcurrency=true
 maxDistance=3000
 soundClassObject=&Fort_Gadgets_Items
 volumeMultiplier=0.6
 concurrencyOverrides {resolutionRule=PreventNew}
}
