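# Sound cue for the Energy Drink consumable's drop sound.
# A Random node picks between two equally weighted bottle-drop waves and feeds a Modulator at default settings, with the cue's overall pitch raised.
# It uses the shared weapon-drop attenuation and the gadget items sound class, and overrides concurrency so new instances are blocked once the limit is reached.
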SoundCue /MusterConsumables/Items/EnergyDrink/Sounds/EnergyDrink_Drop_Cue
root=Modulator_0
nodes [
 {id=Modulator_0, children [Random_0]}
 {id=Random_0, children [WavePlayer_1,WavePlayer_2], properties {weights [1,1]}}
 {id=WavePlayer_1, properties {soundWaveAssetPtr=&/Energy_Bottle_Drop_01}}
 {id=WavePlayer_2, properties {soundWaveAssetPtr=&/Energy_Bottle_Drop_02}}
]
properties {
 attenuationSettings=&Fort_Weapon_Drop_Att
 bOverrideConcurrency=true
 maxDistance=3000
 pitchMultiplier=1.2
 soundClassObject=&Fort_Gadgets_Items
 volumeMultiplier=1
 concurrencyOverrides {resolutionRule=PreventNew}
}
