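# Sound cue for equipping the Energy Drink consumable, playing one of three bottle equip recordings.
# A Random node with equal weights picks the variant, with the first recording trimmed by its own fixed-volume modulator.
# Two stacked modulators on the output add variation and lower the level, alongside a reduced cue volume.
# It uses weapon reload attenuation and the gadgets sound class.
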SoundCue /MusterConsumables/Items/EnergyDrink/Sounds/EnergyDrink_Equip_Cue
root=Modulator_0
nodes [
 {id=Modulator_0, children [Modulator_1]}
 {id=Modulator_1, children [Random_0], properties {volumeMax=0.6, volumeMin=0.5}}
 {id=Modulator_2, children [WavePlayer_3], properties {pitchMax=1, pitchMin=1, volumeMax=0.85, volumeMin=0.85}}
 {id=Random_0, children [Modulator_2,WavePlayer_1,WavePlayer_2], properties {weights [1,1,1]}}
 {id=WavePlayer_1, properties {soundWaveAssetPtr=&/Energy_Bottle_Equip_02}}
 {id=WavePlayer_2, properties {soundWaveAssetPtr=&/Energy_Bottle_Equip_03}}
 {id=WavePlayer_3, properties {soundWaveAssetPtr=&/Energy_Bottle_Equip_01}}
]
properties {
 attenuationSettings=&Fort_Weapon_Reload_Att
 maxDistance=2250
 soundClassObject=&Fort_Gadgets
 volumeMultiplier=0.25
}
