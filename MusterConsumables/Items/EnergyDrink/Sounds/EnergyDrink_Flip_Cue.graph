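# Sound cue for the energy drink consumable's bottle flip, layering a flip sound with a delayed grab sound.
# A mixer blends one of two random flip waves with one of three random grab waves, weighting the grab louder.
# The grab branch has a fixed delay, likely timed to the catch, and a raised pitch.
# A root modulator varies the whole mix, which uses consumable-use attenuation and the gadgets sound class.
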SoundCue /MusterConsumables/Items/EnergyDrink/Sounds/EnergyDrink_Flip_Cue
root=Modulator_0
nodes [
 {id=Delay_0, children [Random_0], properties {delayMax=0.55, delayMin=0.55}}
 {id=Mixer_0, children [Random_1,Modulator_1], properties {inputVolume [0.7,1.4]}}
 {id=Modulator_0, children [Mixer_0]}
 {id=Modulator_1, children [Delay_0], properties {pitchMin=1.05}}
 {id=Random_0, children [WavePlayer_0,WavePlayer_3,WavePlayer_4], properties {weights [1,1,1]}}
 {id=Random_1, children [WavePlayer_2,WavePlayer_6], properties {weights [1,1]}}
 {id=WavePlayer_0, properties {soundWaveAssetPtr=&/Energy_Bottle_Grab_01}}
 {id=WavePlayer_2, properties {soundWaveAssetPtr=&/Energy_Bottle_Flip_01}}
 {id=WavePlayer_3, properties {soundWaveAssetPtr=&/Energy_Bottle_Grab_03}}
 {id=WavePlayer_4, properties {soundWaveAssetPtr=&/Energy_Bottle_Grab_04}}
 {id=WavePlayer_6, properties {soundWaveAssetPtr=&/Energy_Bottle_Flip_02}}
]
properties {
 attenuationSettings=&Fort_Consumable_Use_Athena_Att
 maxDistance=2250
 soundClassObject=&Fort_Gadgets
 volumeMultiplier=0.35
}
