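# Sound cue for the energy drink bottle's landing impact, part of the MusterConsumables item content.
# A Random node picks one of four Energy_Bottle_LandImpact wave variations with equal weight and feeds it through a Modulator at the root.
# It uses the shared weapon-drop attenuation and the loot-collect sound class, with a reduced volume multiplier.
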SoundCue /MusterConsumables/Items/EnergyDrink/Sounds/EnergyDrink_LandImpact_Cue
root=Modulator_0
nodes [
 {id=Modulator_0, children [Random_1]}
 {id=Random_1, children [WavePlayer_0,WavePlayer_3,WavePlayer_4,WavePlayer_5], properties {weights [1,1,1,1]}}
 {id=WavePlayer_0, properties {soundWaveAssetPtr=&/Energy_Bottle_LandImpact_01}}
 {id=WavePlayer_3, properties {soundWaveAssetPtr=&/Energy_Bottle_LandImpact_02}}
 {id=WavePlayer_4, properties {soundWaveAssetPtr=&/Energy_Bottle_LandImpact_03}}
 {id=WavePlayer_5, properties {soundWaveAssetPtr=&/Energy_Bottle_LandImpact_04}}
]
properties {
 attenuationSettings=&Fort_Weapon_Drop_Att
 maxDistance=3000
 soundClassObject=&Fort_Loot_Collect
 volumeMultiplier=0.7
 attenuationOverrides {falloffDistance=2400, stereoSpread=2000, attenuationShapeExtents {x=800}}
}
