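# EnergyDrink_Pickup_Cue is the SoundCue for picking up the Energy Drink consumable item.
# A Random node picks between two equally weighted bottle pickup waves and feeds a Modulator at the root, left at its default settings.
# The cue uses a weapon pickup attenuation asset and the loot collect sound class, with a reduced volume multiplier.
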SoundCue /MusterConsumables/Items/EnergyDrink/Sounds/EnergyDrink_Pickup_Cue
root=Modulator_0
nodes [
 {id=Modulator_0, children [Random_0]}
 {id=Random_0, children [WavePlayer_2,WavePlayer_3], properties {weights [1,1]}}
 {id=WavePlayer_2, properties {soundWaveAssetPtr=&/Energy_Bottle_Pickup_02}}
 {id=WavePlayer_3, properties {soundWaveAssetPtr=&/Energy_Bottle_Pickup_03}}
]
properties {
 attenuationSettings=&Fort_Weapon_Pickup_Att
 maxDistance=3500
 soundClassObject=&Fort_Loot_Collect
 volumeMultiplier=0.42
}
