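# Sound cue for the first-person squirt sound of the Energy Drink consumable, routed through the gadgets sound class with weapon-reload attenuation.
# It randomly picks one of three equally weighted bottle cap impact waves and plays it after a fixed delay of just over half a second.
# The delayed impact feeds a mixer at reduced volume into a root modulator; the mixer's first, louder input is empty.
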SoundCue /MusterConsumables/Items/EnergyDrink/Sounds/EnergyDrink_Squirt_1P_Cue
root=Modulator_0
nodes [
 {id=Delay_0, children [Random_2], properties {delayMax=0.55, delayMin=0.55}}
 {id=Mixer_0, children [null,Delay_0], properties {inputVolume [0.8,0.3]}}
 {id=Modulator_0, children [Mixer_0]}
 {id=Random_2, children [WavePlayer_10, WavePlayer_11, WavePlayer_12], properties {weights [1,1,1]}}
 {id=WavePlayer_10, properties {soundWaveAssetPtr=&/Energy_Bottle_CapImpact_01}}
 {id=WavePlayer_11, properties {soundWaveAssetPtr=&/Energy_Bottle_CapImpact_02}}
 {id=WavePlayer_12, properties {soundWaveAssetPtr=&/Energy_Bottle_CapImpact_03}}
]
properties {
 attenuationSettings=&Fort_Weapon_Reload_Att
 maxDistance=2250
 soundClassObject=&Fort_Gadgets
 volumeMultiplier=0.9
}
