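# Sound cue for squirting the energy drink consumable bottle, with consumable-use attenuation and the gadgets sound class.
# Each branch mixes a random squirt with a quieter, briefly delayed random spray, then applies a Modulator.
# A quality node picks the branch: the first two slots use five squirts and three sprays, the last two use two of each.
# The cue is excluded from random-node branch culling.
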SoundCue /MusterConsumables/Items/EnergyDrink/Sounds/EnergyDrink_Squirt_Cue
root=QualityLevel_0
nodes [
 {id=Delay_0, children [Random_0], properties {delayMax=0.2, delayMin=0.2}}
 {id=Delay_1, children [Random_4], properties {delayMax=0.2, delayMin=0.2}}
 {id=Mixer_0, children [Random_1,Delay_0], properties {inputVolume [0.8,0.3]}}
 {id=Mixer_1, children [Random_3,Delay_1], properties {inputVolume [0.8,0.3]}}
 {id=Modulator_0, children [Mixer_0]}
 {id=Modulator_1, children [Mixer_1]}
 {id=QualityLevel_0, children [Modulator_0,Modulator_0,Modulator_1,Modulator_1]}
 {id=Random_0, children [WavePlayer_0,WavePlayer_1,WavePlayer_3], properties {weights [1,1,1]}}
 {id=Random_1, children [WavePlayer_2,WavePlayer_6,WavePlayer_7,WavePlayer_8,WavePlayer_9], properties {weights [1,1,1,1,1]}}
 {id=Random_3, children [WavePlayer_11, WavePlayer_12], properties {weights [1,1]}}
 {id=Random_4, children [WavePlayer_16, WavePlayer_17], properties {weights [1,1]}}
 {id=WavePlayer_0, properties {soundWaveAssetPtr=&/Energy_Bottle_Spray_A_01}}
 {id=WavePlayer_1, properties {soundWaveAssetPtr=&/Energy_Bottle_Spray_A_02}}
 {id=WavePlayer_11, properties {soundWaveAssetPtr=&/Energy_Bottle_Squirt_01}}
 {id=WavePlayer_12, properties {soundWaveAssetPtr=&/Energy_Bottle_Squirt_02}}
 {id=WavePlayer_16, properties {soundWaveAssetPtr=&Energy_Bottle_Spray_A_01}}
 {id=WavePlayer_17, properties {soundWaveAssetPtr=&Energy_Bottle_Spray_A_02}}
 {id=WavePlayer_2, properties {soundWaveAssetPtr=&Energy_Bottle_Squirt_01}}
 {id=WavePlayer_3, properties {soundWaveAssetPtr=&/Energy_Bottle_Spray_A_03}}
 {id=WavePlayer_6, properties {soundWaveAssetPtr=&Energy_Bottle_Squirt_02}}
 {id=WavePlayer_7, properties {soundWaveAssetPtr=&/Energy_Bottle_Squirt_03}}
 {id=WavePlayer_8, properties {soundWaveAssetPtr=&/Energy_Bottle_Squirt_04}}
 {id=WavePlayer_9, properties {soundWaveAssetPtr=&/Energy_Bottle_Squirt_05}}
]
properties {
 attenuationSettings=&Fort_Consumable_Use_Athena_Att
 bExcludeFromRandomNodeBranchCulling=true
 maxDistance=2250
 soundClassObject=&Fort_Gadgets
 volumeMultiplier=0.8
}
