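# EnergyDrink_Throw_Cue is the sound cue for throwing the energy drink consumable bottle.
# A Random node picks one of five bottle-throw waves with equal weight.
# A Modulator at the root adds slight pitch and volume variation, on top of a raised pitch and reduced volume set on the cue.
# It is routed to the gadgets sound class and spatialized with the shared gadget-throw attenuation.
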SoundCue /MusterConsumables/Items/EnergyDrink/Sounds/EnergyDrink_Throw_Cue
root=Modulator_1
nodes [
 {id=Modulator_1, children [Random_0], properties {pitchMax=1.1, pitchMin=0.9, volumeMax=1.1, volumeMin=0.9}}
 {id=Random_0, children [WavePlayer_0,WavePlayer_1,WavePlayer_3,WavePlayer_4,WavePlayer_5], properties {weights [1,1,1,1,1]}}
 {id=WavePlayer_0, properties {soundWaveAssetPtr=&/Energy_Bottle_Throw_01}}
 {id=WavePlayer_1, properties {soundWaveAssetPtr=&/Energy_Bottle_Throw_02}}
 {id=WavePlayer_3, properties {soundWaveAssetPtr=&/Energy_Bottle_Throw_04}}
 {id=WavePlayer_4, properties {soundWaveAssetPtr=&/Energy_Bottle_Throw_05}}
 {id=WavePlayer_5, properties {soundWaveAssetPtr=&/Energy_Bottle_Throw_06}}
]
properties {
 attenuationSettings=/Game/Sounds/Attenuation/Fort_GadgetThrow_Att
 maxDistance=1536
 pitchMultiplier=1.2
 soundClassObject=&Fort_Gadgets
 volumeMultiplier=0.6
}
